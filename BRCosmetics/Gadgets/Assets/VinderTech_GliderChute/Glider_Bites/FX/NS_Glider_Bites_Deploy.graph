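# Niagara system for the Bites glider, named as its deploy effect: a brief burst flare with a light and, per the emitter name, a shockwave.
# One emitter fires two instantaneous bursts, using spawn group as bone/socket index to keep particles attached to the user-supplied skeletal mesh with an animated offset.
# Lifetime curves drive sprite size, alpha, wave color and dynamic material parameters for the sprite renderer, plus a light renderer's radius.
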NiagaraSystem /BRCosmetics/Gadgets/Assets/VinderTech_GliderChute/Glider_Bites/FX/NS_Glider_Bites_Deploy
dependencies {usedDataInterfaces [&NiagaraDataInterfaceCurve], usedDynamicInputs [], usedModules [&SystemState]}
emitters {FlareLightShockwave {
  emitterUpdateScript [
   &EmitterState(Life Cycle Mode=::NewEnumerator1, Loop Behavior=::NewEnumerator1, Loop Duration Mode=::NewEnumerator1, UseLoopDelay=false) {module=EmitterState001}
   &SpawnBurst_Instantaneous(Use Spawn Probability=false, UseLoopCountLimit=false)
   &SpawnBurst_Instantaneous(Spawn Group=1, Use Spawn Probability=false, UseLoopCountLimit=false) {module=SpawnBurst_Instantaneous001}
  ]
  particleSpawnScript [&InitializeParticle(
    Color=<rgba 0.1598 0.398 1 1>
    Lifetime=0.8
    Mesh Scale Mode=::NewEnumerator3
    Mesh Uniform Scale=&Divide_Float(A=$User.Radius, B=0)
    Position Offset=<15 0 0>
    Sprite Size Mode=::NewEnumerator3
    Uniform Sprite Size=150
    UsePositionOffset=true
    Write Material Random=true
   )]
  particleUpdateScript [
   &ParticleState()
   &CameraOffset(Camera Offset Amount=80) {module=CameraOffset001}
   &SkeletalMeshLocation(
    Bone / Socket Index=&SpawnGroup()
    Bone Sampling Mode=::NewEnumerator4
    MaskSpawnGroup=false
    Orientation Sampling=::NewEnumerator1
    Sampled Position Offset=&MakeVector(
     X=0
     Y=0
     Z=&Add_Float(
      A=&FloatFromCurve(
       FloatCurve=DataInterface() {propertyValues {Curve {keys [
           {interpMode=RCIM_Cubic, tangentMode=RCTM_User, time=0, value=0}
           {arriveTangent=2.7141, interpMode=RCIM_Cubic, leaveTangent=2.7141, tangentMode=RCTM_User, time=1, value=1}
          ]}}}
       Scale Curve=0
      )
      B=10
     )
    )
    Sampled Position Offset Coordinate Space:ENiagaraSkelMeshTransforms=::NewEnumerator2
    Skeletal Mesh=$User.SkeletalMesh
   )
   &DynamicMaterialParameters(
    Index 0 Param 1=1
    Index 0 Param 2=&Lerp_Float(
     A=1
     B=0.8
     Alpha=&FloatFromCurve(FloatCurve=DataInterface() {propertyValues {Curve {keys [
          {arriveTangent=5.1433, interpMode=RCIM_Cubic, time=0, value=0}
          {arriveTangent=0.9449, interpMode=RCIM_Cubic, leaveTangent=0.9449, tangentMode=RCTM_User, time=0.25, value=0.7}
          {arriveTangent=0.6574, interpMode=RCIM_Cubic, leaveTangent=0.6574, tangentMode=RCTM_User, time=1, value=1}
         ]}}})
    )
    Index 3 Param 1=&Multiply_Float(A=&MakeFloatFromVector(Vector=$Particles.Scale, Channel=::NewEnumerator0), B=90)
    Index 3 Param 2=&FloatFromCurve(FloatCurve=DataInterface() {propertyValues {Curve {keys [
         {arriveTangent=2.1437, interpMode=RCIM_Cubic, leaveTangent=2.1437, tangentMode=RCTM_User, time=0, value=0}
         {arriveTangent=1.3259, interpMode=RCIM_Cubic, leaveTangent=1.3259, tangentMode=RCTM_User, time=0.6, value=0.5}
         {interpMode=RCIM_Cubic, leaveTangent=0.2859, time=1, value=1}
        ]}}})
    Module.Param0WriteEnabled=true
    Module.Param0WriteEnabled3=true
    Module.Param1WriteEnabled=true
    Module.Param1WriteEnabled3=true
    Module.Param2WriteEnabled=false
    Module.Param2WriteEnabled3=false
    Module.Param3WriteEnabled=false
    Module.Param3WriteEnabled3=false
    Write Parameter Index 3=true
   )
   &ScaleSpriteSize(Uniform Curve Sprite Scale=DataInterface() {propertyValues {Curve {keys [
        {arriveTangent=0.0561, interpMode=RCIM_Cubic, leaveTangent=0.0561, tangentMode=RCTM_User, time=0, value=0}
        <at 0.2 1>
        {
         arriveTangent=-0.7272
         arriveTangentWeight=0.0565
         interpMode=RCIM_Cubic
         leaveTangent=-0.7272
         leaveTangentWeight=0.1797
         tangentMode=RCTM_User
         time=0.5
         value=0.875
        }
        {interpMode=RCIM_Cubic, leaveTangent=-0.9894, time=0.7, value=0.5}
       ]}}})
   &ScaleColor(
    Scale Alpha=&FloatFromCurve(
     FloatCurve=DataInterface() {propertyValues {Curve {keys [
         {arriveTangent=18.2404, interpMode=RCIM_Cubic, leaveTangent=18.2404, tangentMode=RCTM_User, time=0, value=1}
         {arriveTangent=2.2727, interpMode=RCIM_Cubic, leaveTangent=2.2727, time=0.12, value=2}
         {arriveTangent=-1.1182, interpMode=RCIM_Cubic, leaveTangent=-1.1182, tangentMode=RCTM_User, time=0.44, value=2}
         {interpMode=RCIM_Cubic, tangentMode=RCTM_User, time=1, value=0}
        ]}}}
     Scale Curve=0.25
    )
    ScaleA=true
    ScaleRGB=true
   )
   /SetVariables(Particles.WaveColor:LinearColor=&LerpLinearColors(
     StartColor=<rgba 0.7113 0.6033 1 0>
     EndColor=<rgba 0.5788 1.0477 1.25 1>
     LerpFactor=&FloatFromCurve(FloatCurve=DataInterface() {propertyValues {Curve {keys [
          {arriveTangent=18.2701, interpMode=RCIM_Cubic, time=0, value=0}
          <at 0.15 1>
          {arriveTangent=29.7321, interpMode=RCIM_Cubic, leaveTangent=29.7321, tangentMode=RCTM_User, time=0.27, value=0.5}
          {arriveTangentWeight=0.0997, interpMode=RCIM_Cubic, leaveTangentWeight=0.3488, tangentMode=RCTM_User, time=0.3, value=1.1}
          {interpMode=RCIM_Cubic, leaveTangent=-2.4954, time=1, value=0}
         ]}}})
    )) {isSetParameters=true}
   /SetVariables(Particles.LightColor=&MultiplyLinearColor(Linear Color=$Particles.Color, Scale Factor=&MakeVector4(X=1, Y=1, Z=1, W=15))) {isSetParameters=true}
   /SetVariables(Particles.LightRadius=&FloatFromCurve(FloatCurve=DataInterface() {propertyValues {Curve {keys [<at 0.5 1>]}}}, Scale Curve=150)) {isSetParameters=true}
  ]
  rendererClasses [&NiagaraSpriteRendererProperties, &NiagaraLightRendererProperties]
  renderers [
   {rendererClass=&NiagaraSpriteRendererProperties}
   {rendererClass=&NiagaraLightRendererProperties, rendererIndex=1}
  ]
 }}
summary {
 systemName=NS_Glider_Bites_Deploy
 emitters [{
   emitterName=FlareLightShockwave
   rendererClasses [&NiagaraSpriteRendererProperties, &NiagaraLightRendererProperties]
  }]
 userVariables [
  User.BoneFlowSK {
   type=NiagaraDataInterfaceSkeletalMesh
   defaultValue {
    struct=NiagaraExt_VariableValue_DataInterface
    value {dataInterface=&self:NiagaraDataInterfaceSkeletalMesh_0}
   }
  }
  User.Radius {type=NiagaraFloat, defaultValue=0, description=""}
  User.RibbonIDOffset {type=NiagaraInt32}
  User.SkeletalMesh {
   type=NiagaraDataInterfaceSkeletalMesh
   defaultValue {
    struct=NiagaraExt_VariableValue_DataInterface
    value {dataInterface=&self:NiagaraDataInterfaceSkeletalMesh_1}
   }
  }
 ]
}
