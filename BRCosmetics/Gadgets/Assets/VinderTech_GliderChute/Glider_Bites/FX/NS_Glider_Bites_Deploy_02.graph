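# Niagara deploy effect for the Bites glider, built from mirrored left and right emitters for blobs, flares, ribbons and tendrils.
# Blob sprites, flares and ribbons follow socket chains on the glider skeletal mesh through FN_BoneFlow, with noisy jitter on blobs and ribbons facing along the spline binormal.
# Tendril meshes burst at sockets on User.TendrilsSKM and flip X in the front end, where the StartBlobs burst spawns nothing.
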NiagaraSystem /BRCosmetics/Gadgets/Assets/VinderTech_GliderChute/Glider_Bites/FX/NS_Glider_Bites_Deploy_02
dependencies {
 usedDataInterfaces [&NiagaraDataInterfaceCamera, &NiagaraDataInterfaceCurve]
 usedDynamicInputs []
 usedModules [&SystemState]
}
emitters {
 Left_Blobs002 {
  emitterUpdateScript [
   &EmitterState()
   &SpawnRate(SpawnRate=150, Use Spawn Probability=false)
  ]
  particleSpawnScript [
   /Niagara/Modules/Spawn/Initialization/InitializeParticle(
    Color=<rgba 0.276 0.276 0.276 1>
    Lifetime=&RandomRangeFloat(Minimum=1, Maximum=1.4)
    Module.Write Color=true
    Module.Write Lifetime=true
    Module.Write Mass=true
    Module.Write Material Random=true
    Module.Write Position=false
    Module.Write Scale=false
    Module.Write SpriteRotation=false
    Module.Write SpriteSize=true
    Position=&SimulationPosition()
    Sprite Size=&Vector2DFromFloat(Value=&RandomRangeFloat(Minimum=6.5, Maximum=10))
   )
   &FN_SetSpawnGroup(Normalize Spawn Group=true, Number of Spawn Groups=11, Particles Per Group=300, Set Ribbon ID and Ribbon Link Order=true)
  ]
  particleUpdateScript [
   &ParticleState()
   /Game/Effects/Niagara/Modules/BoneFlow/FN_BoneFlow_v4/FN_BoneFlow(
    Angle:NiagaraFloat=&RandomRangeFloat(Minimum=0, Maximum=0)
    BoneID:NiagaraInt32=&ModuloInt(A=$Particles.SpawnGroup, B=11)
    Height:NiagaraFloat=&FloatFromCurve(FloatCurve=DataInterface() {propertyValues {Curve {keys [{interpMode=RCIM_Cubic, tangentMode=RCTM_User, time=0, value=0}, {interpMode=RCIM_Cubic, tangentMode=RCTM_User, time=1, value=1}]}}})
    SkeletalMesh=DataInterface() {propertyValues {
      DefaultMesh=null
      ExcludeBoneName=root
      OptionalComponentClass=null
      PreviewMesh=/BRCosmetics/Gadgets/Assets/VinderTech_GliderChute/Glider_Bites/Meshes/Bites_Glider
      SkinningMode=SkinOnTheFly
      SoftSourceActor=null
      SourceMode=Default
      UvSetIndex=0
      WholeMeshLOD=-1
      bExcludeBone=false
      bReadDeformedGeometry=true
      bRequireCurrentFrameData=true
      ComponentTags []
      FilteredBones [JNT_lf_Base_root, JNT_lf_Base_4, JNT_lf_Branch4_3]
      FilteredSockets [
       L_Socket_Base
       L_Socket_A_01
       L_Socket_A_02
       L_Socket_Base
       L_Socket_B_01
       L_Socket_B_02
       L_Socket_Base
       L_Socket_C_01
       L_Socket_C_02
       L_Socket_Base
       L_Socket_D_01
       L_Socket_D_02
       L_Socket_Base
       L_Socket_E_01
       L_Socket_E_02
       L_Socket_Base
       L_Socket_F_01
       L_Socket_F_02
       L_Socket_B_01
       L_Socket_G_01
       L_Socket_G_02
       L_Socket_B_01
       L_Socket_H_01
       L_Socket_H_02
       L_Socket_B_01
       L_Socket_I_01
       L_Socket_I_02
       L_Socket_Base
       L_Socket_J_01
       L_Socket_J_02
       L_Socket_D_01
       L_Socket_K_01
       L_Socket_K_02
      ]
      MeshUserParameter {parameter {name=None, typeDefHandle {registeredTypeIndex=1}, varData []}}
      SamplingRegions []
     }}
    Use Sockets Instead=true
    Write Position Attribute=true
   ) {module=FN_BoneFlow002}
   /SetVariables(Particles.Position=&Add_Vector(
     A=Linked() {linkedVariable {name=Particles.Position, type=NiagaraPosition}}
     B=&Lerp_Vector(
      A=<0 0 0>
      B=&RandomRangeVector(Minimum=<-15 -15 -15>, Maximum=<15 15 15>)
      Alpha=&FloatFromCurve(FloatCurve=DataInterface() {propertyValues {Curve {keys [
           {interpMode=RCIM_Cubic, tangentMode=RCTM_User, time=0, value=0.5}
           {interpMode=RCIM_Cubic, tangentMode=RCTM_User, time=0.5, value=1}
           {tangentMode=RCTM_User, time=1, value=0.5}
          ]}}})
     )
    )) {isSetParameters=true, module=SetVariables_#1}
   /Niagara/Modules/Update/Forces/CurlNoiseForce(
    Noise Frequency=0.35
    Noise Strength=&FloatFromCurve(FloatCurve=DataInterface() {propertyValues {Curve {keys [<at 0 1 RCIM_Cubic>,<at 0.5 0 RCIM_Cubic>]}}}, Scale Curve=30)
    Pan Noise=true
    Pan Noise Field=<0 0 2>
   )
   &SolveForcesAndVelocity(Clamp Velocity=true, Limit Acceleration=false, Speed Limit=1)
   &ScaleSpriteSize(Scale Factor=&Vector2DFromFloat(Value=&FloatFromCurve(FloatCurve=DataInterface() {propertyValues {Curve {keys [
          {arriveTangent=-0.0558, interpMode=RCIM_Cubic, leaveTangent=-0.0558, tangentMode=RCTM_User, time=0, value=0}
          {arriveTangent=0.044, interpMode=RCIM_Cubic, leaveTangent=0.044, tangentMode=RCTM_User, time=0.2, value=0.9847}
          {arriveTangent=-1.2308, interpMode=RCIM_Cubic, leaveTangent=-1.2308, time=0.7, value=0.8}
          {arriveTangent=-0.0405, interpMode=RCIM_Cubic, leaveTangent=-0.0405, tangentMode=RCTM_User, time=1, value=0}
         ]}}})))
   &CameraOffset(Camera Offset Amount=&RandomRangeFloat(Minimum=0, Maximum=5))
   &DynamicMaterialParameters(
    Index 0 Param 1=&FloatFromCurve(FloatCurve=DataInterface() {propertyValues {Curve {keys [
         {arriveTangent=130.5352, interpMode=RCIM_Cubic, leaveTangent=130.5352, tangentMode=RCTM_User, time=0, value=0}
         {arriveTangent=-0.891, interpMode=RCIM_Cubic, leaveTangent=-0.891, tangentMode=RCTM_User, time=0.03, value=1.7}
         <at 0.2 0.75>
         {tangentMode=RCTM_User, time=1, value=0}
        ]}}})
    Index 0 Param 2=5
    Index 0 Param 3=30
    Index 0 Param 4=7
    Index 1 Param 1=&FloatFromCurve(
     FloatCurve=DataInterface() {propertyValues {Curve {keys [
         {arriveTangent=-3.7492, interpMode=RCIM_Cubic, leaveTangent=-3.7492, tangentMode=RCTM_User, time=0, value=1}
         {arriveTangent=-0.7477, interpMode=RCIM_Cubic, leaveTangent=-0.7477, tangentMode=RCTM_User, time=0.3, value=0.3}
         {interpMode=RCIM_Cubic, leaveTangent=-1.9286, time=1, value=0}
        ]}}}
     CurveIndex=&FloatFromCurve(FloatCurve=DataInterface() {propertyValues {Curve {keys [
          {arriveTangent=130.5352, interpMode=RCIM_Cubic, leaveTangent=130.5352, tangentMode=RCTM_User, time=0, value=0}
          {arriveTangent=-0.891, interpMode=RCIM_Cubic, leaveTangent=-0.891, tangentMode=RCTM_User, time=0.03, value=1.7}
          <at 0.2 0.75>
          {tangentMode=RCTM_User, time=1, value=0}
         ]}}})
     Scale Curve=&RandomRangeFloat(Minimum=4, Maximum=4.5)
    )
    Module.Param0WriteEnabled=true
    Module.Param0WriteEnabled1=true
    Module.Param1WriteEnabled=true
    Module.Param1WriteEnabled1=false
    Module.Param2WriteEnabled=true
    Module.Param2WriteEnabled1=false
    Module.Param3WriteEnabled=true
    Module.Param3WriteEnabled1=false
    Write Parameter Index 1=true
   )
   /SetVariables(Particles.Velocity=&Multiply_VectorByFloat(Vector=$Output.FN_BoneFlow002.SplineTangent, Float=30)) {isSetParameters=true, module=SetVariables_#2}
  ]
  rendererClasses [&NiagaraSpriteRendererProperties]
  renderers [{rendererClass=&NiagaraSpriteRendererProperties}]
 }
 Left_Flares {
  emitterUpdateScript [
   &EmitterState(Life Cycle Mode=::NewEnumerator1, Loop Behavior=::NewEnumerator1, Loop Duration Mode=::NewEnumerator1, UseLoopDelay=false) {module=EmitterState001}
   &SpawnBurst_Instantaneous(Spawn Count=4, Use Spawn Probability=false, UseLoopCountLimit=false)
  ]
  particleSpawnScript [
   /Niagara/Modules/Spawn/Initialization/V2/InitializeParticle(
    Color=<rgba 0.1598 0.398 1 1>
    Lifetime=1.2
    Mesh Scale Mode=::NewEnumerator3
    Mesh Uniform Scale=&Divide_Float(A=$User.Radius, B=0)
    Position Offset=<15 0 0>
    Sprite Size Mode=::NewEnumerator3
    Uniform Sprite Size=50
    UsePositionOffset=false
    Write Material Random=true
   )
   &FN_SetSpawnGroup()
  ]
  particleUpdateScript [
   &ParticleState()
   &FN_BoneFlow(
    BoneID:NiagaraInt32=&ModuloInt(A=$Particles.SpawnGroup, B=4)
    Height:NiagaraFloat=&FloatFromCurve(FloatCurve=DataInterface() {propertyValues {Curve {keys [{interpMode=RCIM_Cubic, tangentMode=RCTM_User, time=0, value=0}, {interpMode=RCIM_Cubic, tangentMode=RCTM_User, time=1, value=1}]}}})
    SkeletalMesh=DataInterface() {propertyValues {
      DefaultMesh=null
      ExcludeBoneName=root
      OptionalComponentClass=null
      PreviewMesh=&Bites_Glider
      SkinningMode=SkinOnTheFly
      SoftSourceActor=null
      SourceMode=Default
      UvSetIndex=0
      WholeMeshLOD=-1
      bExcludeBone=false
      bReadDeformedGeometry=true
      bRequireCurrentFrameData=true
      ComponentTags []
      FilteredBones [JNT_lf_Base_root, JNT_lf_Base_4, JNT_lf_Branch4_3]
      FilteredSockets [
       L_Socket_Base
       L_Socket_A_01
       L_Socket_A_02
       L_Socket_Base
       L_Socket_B_01
       L_Socket_B_02
       L_Socket_Base
       L_Socket_C_01
       L_Socket_C_02
       L_Socket_Base
       L_Socket_D_01
       L_Socket_D_02
      ]
      MeshUserParameter {parameter {name=None, typeDefHandle {registeredTypeIndex=1}, varData []}}
      SamplingRegions []
     }}
    Use Sockets Instead=true
    Write Position Attribute=true
   ) {module=FN_BoneFlow002}
   &CameraOffset(Camera Offset Amount=80) {module=CameraOffset001}
   &DynamicMaterialParameters(
    Index 0 Param 1=1
    Index 0 Param 2=&Lerp_Float(
     A=1
     B=0.8
     Alpha=&FloatFromCurve(FloatCurve=DataInterface() {propertyValues {Curve {keys [
          {arriveTangent=5.1433, interpMode=RCIM_Cubic, time=0, value=0}
          {arriveTangent=0.9449, interpMode=RCIM_Cubic, leaveTangent=0.9449, tangentMode=RCTM_User, time=0.25, value=0.7}
          {arriveTangent=0.6574, interpMode=RCIM_Cubic, leaveTangent=0.6574, tangentMode=RCTM_User, time=1, value=1}
         ]}}})
    )
    Index 3 Param 1=&Multiply_Float(A=&MakeFloatFromVector(Vector=$Particles.Scale, Channel=::NewEnumerator0), B=90)
    Index 3 Param 2=&FloatFromCurve(FloatCurve=DataInterface() {propertyValues {Curve {keys [
         {arriveTangent=2.1437, interpMode=RCIM_Cubic, leaveTangent=2.1437, tangentMode=RCTM_User, time=0, value=0}
         {arriveTangent=1.3259, interpMode=RCIM_Cubic, leaveTangent=1.3259, tangentMode=RCTM_User, time=0.6, value=0.5}
         {interpMode=RCIM_Cubic, leaveTangent=0.2859, time=1, value=1}
        ]}}})
    Module.Param0WriteEnabled=true
    Module.Param0WriteEnabled3=true
    Module.Param1WriteEnabled=true
    Module.Param1WriteEnabled3=true
    Module.Param2WriteEnabled=false
    Module.Param2WriteEnabled3=false
    Module.Param3WriteEnabled=false
    Module.Param3WriteEnabled3=false
    Write Parameter Index 3=true
   )
   &ScaleSpriteSize(Uniform Curve Sprite Scale=DataInterface() {propertyValues {Curve {keys [
        {arriveTangent=0.0561, interpMode=RCIM_Cubic, leaveTangent=0.0561, tangentMode=RCTM_User, time=0, value=0}
        <at 0.2 1>
        {
         arriveTangent=-0.7272
         arriveTangentWeight=0.0565
         interpMode=RCIM_Cubic
         leaveTangent=-0.7272
         leaveTangentWeight=0.1797
         tangentMode=RCTM_User
         time=0.5
         value=0.875
        }
        {interpMode=RCIM_Cubic, leaveTangent=-0.9894, time=1, value=0.65}
       ]}}})
   &ScaleColor(
    Scale Alpha=&FloatFromCurve(
     FloatCurve=DataInterface() {propertyValues {Curve {keys [
         {arriveTangent=18.2404, interpMode=RCIM_Cubic, leaveTangent=18.2404, tangentMode=RCTM_User, time=0, value=1}
         {arriveTangent=-0.1486, interpMode=RCIM_Cubic, leaveTangent=-0.1486, tangentMode=RCTM_User, time=0.12, value=2}
         {arriveTangent=-0.1686, interpMode=RCIM_Cubic, leaveTangent=-0.1686, tangentMode=RCTM_User, time=0.7, value=1.875}
         {arriveTangent=-0.3515, interpMode=RCIM_Cubic, leaveTangent=-0.3515, tangentMode=RCTM_User, time=1, value=0}
        ]}}}
     Scale Curve=0.5
    )
    ScaleA=true
    ScaleRGB=true
   )
   /SetVariables(Particles.WaveColor:LinearColor=&LerpLinearColors(
     StartColor=<rgba 0.7113 0.6033 1 0>
     EndColor=<rgba 0.5788 1.0477 1.25 1>
     LerpFactor=&FloatFromCurve(FloatCurve=DataInterface() {propertyValues {Curve {keys [
          {arriveTangent=18.2701, interpMode=RCIM_Cubic, time=0, value=0}
          <at 0.15 1>
          {arriveTangent=29.7321, interpMode=RCIM_Cubic, leaveTangent=29.7321, tangentMode=RCTM_User, time=0.27, value=0.5}
          {arriveTangentWeight=0.0997, interpMode=RCIM_Cubic, leaveTangentWeight=0.3488, tangentMode=RCTM_User, time=0.3, value=1.1}
          {interpMode=RCIM_Cubic, leaveTangent=-2.4954, time=1, value=0}
         ]}}})
    )) {isSetParameters=true, module=SetVariables_#3}
   /SetVariables(Particles.LightColor=&MultiplyLinearColor(Linear Color=$Particles.Color, Scale Factor=&MakeVector4(X=1, Y=1, Z=1, W=15))) {isSetParameters=true, module=SetVariables_#4}
   /SetVariables(Particles.LightRadius=&FloatFromCurve(FloatCurve=DataInterface() {propertyValues {Curve {keys [<at 0.2 0.35 RCIM_Cubic>,<at 0.5 1>]}}}, Scale Curve=75)) {isSetParameters=true, module=SetVariables_#5}
  ]
  rendererClasses [&NiagaraSpriteRendererProperties]
  renderers [{rendererClass=&NiagaraSpriteRendererProperties}]
 }
 Left_Ribbons {
  emitterUpdateScript [
   &EmitterState(Life Cycle Mode=::NewEnumerator1, Loop Behavior=::NewEnumerator1, Loop Delay=0.5, Loop Duration=0.3, UseLoopDelay=true)
   &SpawnRate(SpawnRate=&Multiply_Float(A=30, B=20), Use Spawn Probability=false)
  ]
  particleSpawnScript [
   &FN_SetSpawnGroup(Inclusivity=true, Normalize Spawn Group=true, Particles Per Group=20, Pattern=::NewEnumerator0, Set Ribbon ID and Ribbon Link Order=true)
   &InitializeRibbon(
    Ribbon Width=&Multiply_Float(A=&RandomRangeFloat(Minimum=2.5, Maximum=3.5), B=2)
    UsePositionOffset=false
    Write Material Random=true
   )
  ]
  particleUpdateScript [
   &ParticleState()
   &FN_BoneFlow(
    Angle:NiagaraFloat=&RandomRangeFloat(Minimum=0, Maximum=1)
    BoneID:NiagaraInt32=&ModuloInt(A=$Particles.SpawnGroup, B=4)
    Height:NiagaraFloat=&Multiply_Float(A=$Particles.Normalized Spawn Group, B=1)
    Radius:NiagaraFloat=&RandomRangeFloat(Minimum=0, Maximum=10)
    SkeletalMesh=DataInterface() {propertyValues {
      DefaultMesh=null
      ExcludeBoneName=root
      OptionalComponentClass=null
      PreviewMesh=&Bites_Glider
      SkinningMode=SkinOnTheFly
      SoftSourceActor=null
      SourceMode=Default
      UvSetIndex=0
      WholeMeshLOD=-1
      bExcludeBone=false
      bReadDeformedGeometry=true
      bRequireCurrentFrameData=true
      ComponentTags []
      FilteredBones [JNT_lf_Base_root, JNT_lf_Base_4, JNT_lf_Branch4_3]
      FilteredSockets [
       L_Socket_Base
       L_Socket_A_01
       L_Socket_A_02
       L_Socket_Base
       L_Socket_B_01
       L_Socket_B_02
       L_Socket_Base
       L_Socket_C_01
       L_Socket_C_02
       L_Socket_Base
       L_Socket_D_01
       L_Socket_D_02
      ]
      MeshUserParameter {parameter {name=None, typeDefHandle {registeredTypeIndex=1}, varData []}}
      SamplingRegions []
     }}
    Use Sockets Instead=true
    Write Position Attribute=true
   ) {module=FN_BoneFlow001}
   /SetVariables(Particles.Position=&Add_Vector(
     A=Linked() {linkedVariable {name=Particles.Position, type=NiagaraPosition}}
     B=&Multiply_Vector(A=&ReturnCameraVector(Camera Vectors=::NewEnumerator1, Coordinate Space=::Simulation), B=<0 0 0>)
    )) {isSetParameters=true, module=SetVariables_#6}
   &ScaleRibbonWidth(Ribbon Width Scale=&FloatFromCurve(
     FloatCurve=DataInterface() {propertyValues {Curve {keys [
         {interpMode=RCIM_Cubic, tangentMode=RCTM_User, time=0, value=0}
         {arriveTangent=-5.6849, interpMode=RCIM_Cubic, leaveTangent=-5.6849, tangentMode=RCTM_User, time=0.1, value=0.4}
         {arriveTangent=-5.4632, interpMode=RCIM_Cubic, leaveTangent=-5.4632, tangentMode=RCTM_User, time=0.5, value=0.4}
         {arriveTangent=-2.6171, interpMode=RCIM_Cubic, leaveTangent=-2.6171, tangentMode=RCTM_User, time=0.8, value=0.5}
         {arriveTangent=-0.2025, interpMode=RCIM_Cubic, leaveTangent=-0.2025, tangentMode=RCTM_User, time=1, value=0.025}
        ]}}}
     CurveIndex=$Particles.Normalized Spawn Group
     Scale Curve=&Multiply_Float(
      A=&FloatFromCurve(
       FloatCurve=DataInterface() {propertyValues {Curve {keys [
           {arriveTangent=-0.082, interpMode=RCIM_Cubic, leaveTangent=-0.082, tangentMode=RCTM_User, time=0, value=0}
           {interpMode=RCIM_Cubic, tangentMode=RCTM_User, time=0.2, value=1.05}
           {interpMode=RCIM_Cubic, tangentMode=RCTM_User, time=0.5, value=1.2}
           {arriveTangent=-1.5059, interpMode=RCIM_Cubic, leaveTangent=-1.5059, tangentMode=RCTM_User, time=0.99, value=0}
          ]}}}
       Scale Curve=4
      )
      B=1
     )
    ))
   /SetVariables(Particles.RibbonFacing=$Output.FN_BoneFlow001.SplineBinormal) {isSetParameters=true, module=SetVariables_#7}
   &DynamicMaterialParameters(
    Index 0 Param 1=&FloatFromCurve(FloatCurve=DataInterface() {propertyValues {Curve {keys [<at 0 0 RCIM_Cubic>,<at 0.5 1>,<at 1 0 RCIM_Cubic>]}}})
    Module.Param0WriteEnabled=true
    Module.Param1WriteEnabled=true
    Module.Param2WriteEnabled=false
    Module.Param3WriteEnabled=false
   )
   &SolveForcesAndVelocity(Clamp Velocity=false, Limit Acceleration=false)
  ]
  rendererClasses [&NiagaraRibbonRendererProperties]
  renderers [{rendererClass=&NiagaraRibbonRendererProperties}]
 }
 Left_Tendrils {
  emitterUpdateScript [
   &EmitterState()
   &SpawnBurst_Instantaneous(Spawn Time=0.3, Use Spawn Probability=false, UseLoopCountLimit=false)
   &SpawnBurst_Instantaneous(Spawn Group=1, Spawn Time=0.25, Use Spawn Probability=false, UseLoopCountLimit=false) {module=SpawnBurst_Instantaneous001}
  ]
  particleSpawnScript [
   &InitializeParticle(
    Mesh Scale Mode=::NewEnumerator3
    Mesh Uniform Scale=&MaskFloatBySpawnGroup(Masked Float=1.25, Passthrough Float=1.5)
    UsePositionOffset=false
    Write Material Random=true
   )
   &InitialMeshOrientation(Mesh Orientation Mode=::NewEnumerator2, Use Rotation Vector=true) {module=InitialMeshOrientation002}
   /SetVariables(Particles.VisibilityTag=&MaskIntBySpawnGroup(Masked Int=0, Passthrough Int=1, Spawn Group Mask=0)) {isSetParameters=true, module=SetVariables_#8}
  ]
  particleUpdateScript [
   &ParticleState()
   &ScaleMeshSize(Scale Factor=&FN_FrontEndVector3Modifier(Gameplay=<1 1 1>, Front End=<-1 1 1>))
   &ScaleMeshSize(Scale Factor=&VectorFromFloat(Value=&FloatFromCurve(FloatCurve=DataInterface() {propertyValues {Curve {keys [
          {arriveTangent=-0.0856, interpMode=RCIM_Cubic, leaveTangent=-0.0856, tangentMode=RCTM_User, time=0, value=0.8}
          {arriveTangent=-0.0405, interpMode=RCIM_Cubic, leaveTangent=-0.0405, tangentMode=RCTM_User, time=0.3, value=1.35}
         ]}}}))) {module=ScaleMeshSize001}
   &SkeletalMeshLocation(
    MaskSpawnGroup=false
    Mesh Sampling Type=::NewEnumerator5
    Orientation Sampling=::NewEnumerator1
    Sampled Position Offset=<-10 10 0>
    Skeletal Mesh=$User.TendrilsSKM
    Socket Sampling Mode=::NewEnumerator5
   )
   &CameraOffset(Camera Offset Amount=20)
   &DynamicMaterialParameters(
    Index 0 Param 1=&FloatFromCurve(FloatCurve=DataInterface() {propertyValues {Curve {keys [{arriveTangent=2.301, interpMode=RCIM_Cubic, time=0.8, value=0}, {interpMode=RCIM_Cubic, tangentMode=RCTM_User, time=1, value=1}]}}})
    Index 0 Param 3=1
    Index 1 Param 2=1
    Index 2 Param 1=&FloatFromCurve(FloatCurve=DataInterface() {propertyValues {Curve {keys [{tangentMode=RCTM_User, time=0, value=-1}, {tangentMode=RCTM_User, time=1, value=1}]}}})
    Index 2 Param 3=1
    Index 2 Param 4=2
    Module.Param0WriteEnabled=true
    Module.Param0WriteEnabled1=true
    Module.Param0WriteEnabled2=true
    Module.Param1WriteEnabled=true
    Module.Param1WriteEnabled1=true
    Module.Param1WriteEnabled2=true
    Module.Param2WriteEnabled=true
    Module.Param2WriteEnabled1=false
    Module.Param2WriteEnabled2=true
    Module.Param3WriteEnabled=true
    Module.Param3WriteEnabled1=false
    Module.Param3WriteEnabled2=true
    Write Parameter Index 1=true
    Write Parameter Index 2=true
   ) {module=DynamicMaterialParameters001}
  ]
  rendererClasses [&NiagaraMeshRendererProperties]
  renderers [
   {rendererClass=&NiagaraMeshRendererProperties, rendererIndex=0}
   {rendererClass=&NiagaraMeshRendererProperties, rendererIndex=1}
  ]
 }
 Right_Blobs003 {
  emitterUpdateScript [
   &EmitterState()
   &SpawnRate(SpawnRate=250, Use Spawn Probability=false)
  ]
  particleSpawnScript [
   /Niagara/Modules/Spawn/Initialization/InitializeParticle(
    Color=<rgba 0.276 0.276 0.276 1>
    Lifetime=&RandomRangeFloat(Minimum=1, Maximum=1.4)
    Module.Write Color=true
    Module.Write Lifetime=true
    Module.Write Mass=true
    Module.Write Material Random=true
    Module.Write Position=false
    Module.Write Scale=false
    Module.Write SpriteRotation=false
    Module.Write SpriteSize=true
    Position=&SimulationPosition()
    Sprite Size=&Vector2DFromFloat(Value=&RandomRangeFloat(Minimum=6.5, Maximum=10))
   )
   &FN_SetSpawnGroup(Normalize Spawn Group=true, Number of Spawn Groups=11, Particles Per Group=300, Set Ribbon ID and Ribbon Link Order=true)
  ]
  particleUpdateScript [
   &ParticleState()
   &FN_BoneFlow(
    Angle:NiagaraFloat=&RandomRangeFloat(Minimum=0, Maximum=0)
    BoneID:NiagaraInt32=&ModuloInt(A=$Particles.SpawnGroup, B=11)
    Height:NiagaraFloat=&FloatFromCurve(FloatCurve=DataInterface() {propertyValues {Curve {keys [{interpMode=RCIM_Cubic, tangentMode=RCTM_User, time=0, value=0}, {interpMode=RCIM_Cubic, tangentMode=RCTM_User, time=1, value=1}]}}})
    SkeletalMesh=DataInterface() {propertyValues {
      DefaultMesh=null
      ExcludeBoneName=root
      OptionalComponentClass=null
      PreviewMesh=&Bites_Glider
      SkinningMode=SkinOnTheFly
      SoftSourceActor=null
      SourceMode=Default
      UvSetIndex=0
      WholeMeshLOD=-1
      bExcludeBone=false
      bReadDeformedGeometry=true
      bRequireCurrentFrameData=true
      ComponentTags []
      FilteredBones [JNT_lf_Base_root, JNT_lf_Base_4, JNT_lf_Branch4_3]
      FilteredSockets [
       R_Socket_Base
       R_Socket_A_01
       R_Socket_A_02
       R_Socket_Base
       R_Socket_B_01
       R_Socket_B_02
       R_Socket_Base
       R_Socket_C_01
       R_Socket_C_02
       R_Socket_Base
       R_Socket_D_01
       R_Socket_D_02
       R_Socket_Base
       R_Socket_E_01
       R_Socket_E_02
       R_Socket_Base
       R_Socket_F_01
       R_Socket_F_02
       R_Socket_B_01
       R_Socket_G_01
       R_Socket_G_02
       R_Socket_B_01
       R_Socket_H_01
       R_Socket_H_02
       R_Socket_B_01
       R_Socket_I_01
       R_Socket_I_02
       R_Socket_Base
       R_Socket_J_01
       R_Socket_J_02
       R_Socket_D_01
       R_Socket_K_01
       R_Socket_K_02
      ]
      MeshUserParameter {parameter {name=None, typeDefHandle {registeredTypeIndex=1}, varData []}}
      SamplingRegions []
     }}
    Use Sockets Instead=true
    Write Position Attribute=true
   ) {module=FN_BoneFlow002}
   /SetVariables(Particles.Position=&Add_Vector(
     A=Linked() {linkedVariable {name=Particles.Position, type=NiagaraPosition}}
     B=&Lerp_Vector(
      A=<0 0 0>
      B=&RandomRangeVector(Minimum=<-15 -15 -15>, Maximum=<15 15 15>)
      Alpha=&FloatFromCurve(FloatCurve=DataInterface() {propertyValues {Curve {keys [
           {interpMode=RCIM_Cubic, tangentMode=RCTM_User, time=0, value=0.5}
           {interpMode=RCIM_Cubic, tangentMode=RCTM_User, time=0.5, value=1}
           {tangentMode=RCTM_User, time=1, value=0.5}
          ]}}})
     )
    )) {isSetParameters=true, module=SetVariables_#1}
   &CurlNoiseForce(
    Noise Frequency=0.35
    Noise Strength=&FloatFromCurve(FloatCurve=DataInterface() {propertyValues {Curve {keys [<at 0 1 RCIM_Cubic>,<at 0.5 0 RCIM_Cubic>]}}}, Scale Curve=30)
    Pan Noise=true
    Pan Noise Field=<0 0 2>
   )
   &SolveForcesAndVelocity(Clamp Velocity=true, Limit Acceleration=false, Speed Limit=1)
   &ScaleSpriteSize(Scale Factor=&Vector2DFromFloat(Value=&FloatFromCurve(FloatCurve=DataInterface() {propertyValues {Curve {keys [
          {arriveTangent=-0.0558, interpMode=RCIM_Cubic, leaveTangent=-0.0558, tangentMode=RCTM_User, time=0, value=0}
          {arriveTangent=0.044, interpMode=RCIM_Cubic, leaveTangent=0.044, tangentMode=RCTM_User, time=0.2, value=0.9847}
          {arriveTangent=-1.2308, interpMode=RCIM_Cubic, leaveTangent=-1.2308, time=0.7, value=0.8}
          {arriveTangent=-0.0405, interpMode=RCIM_Cubic, leaveTangent=-0.0405, tangentMode=RCTM_User, time=1, value=0}
         ]}}})))
   &CameraOffset(Camera Offset Amount=&RandomRangeFloat(Minimum=0, Maximum=5))
   &DynamicMaterialParameters(
    Index 0 Param 1=&FloatFromCurve(FloatCurve=DataInterface() {propertyValues {Curve {keys [
         {arriveTangent=130.5352, interpMode=RCIM_Cubic, leaveTangent=130.5352, tangentMode=RCTM_User, time=0, value=0}
         {arriveTangent=-0.891, interpMode=RCIM_Cubic, leaveTangent=-0.891, tangentMode=RCTM_User, time=0.03, value=1.7}
         <at 0.2 0.75>
         {tangentMode=RCTM_User, time=1, value=0}
        ]}}})
    Index 0 Param 2=5
    Index 0 Param 3=30
    Index 0 Param 4=7
    Index 1 Param 1=&FloatFromCurve(
     FloatCurve=DataInterface() {propertyValues {Curve {keys [
         {arriveTangent=-3.7492, interpMode=RCIM_Cubic, leaveTangent=-3.7492, tangentMode=RCTM_User, time=0, value=1}
         {arriveTangent=-0.7477, interpMode=RCIM_Cubic, leaveTangent=-0.7477, tangentMode=RCTM_User, time=0.3, value=0.3}
         {interpMode=RCIM_Cubic, leaveTangent=-1.9286, time=1, value=0}
        ]}}}
     CurveIndex=&FloatFromCurve(FloatCurve=DataInterface() {propertyValues {Curve {keys [
          {arriveTangent=130.5352, interpMode=RCIM_Cubic, leaveTangent=130.5352, tangentMode=RCTM_User, time=0, value=0}
          {arriveTangent=-0.891, interpMode=RCIM_Cubic, leaveTangent=-0.891, tangentMode=RCTM_User, time=0.03, value=1.7}
          <at 0.2 0.75>
          {tangentMode=RCTM_User, time=1, value=0}
         ]}}})
     Scale Curve=&RandomRangeFloat(Minimum=4, Maximum=4.5)
    )
    Module.Param0WriteEnabled=true
    Module.Param0WriteEnabled1=true
    Module.Param1WriteEnabled=true
    Module.Param1WriteEnabled1=false
    Module.Param2WriteEnabled=true
    Module.Param2WriteEnabled1=false
    Module.Param3WriteEnabled=true
    Module.Param3WriteEnabled1=false
    Write Parameter Index 1=true
   )
   /SetVariables(Particles.Velocity=&Multiply_VectorByFloat(Vector=$Output.FN_BoneFlow002.SplineTangent, Float=30)) {isSetParameters=true, module=SetVariables_#2}
  ]
  rendererClasses [&NiagaraSpriteRendererProperties]
  renderers [{rendererClass=&NiagaraSpriteRendererProperties}]
 }
 Right_Flares {
  emitterUpdateScript [
   &EmitterState(Life Cycle Mode=::NewEnumerator1, Loop Behavior=::NewEnumerator1, Loop Duration Mode=::NewEnumerator1, UseLoopDelay=false) {module=EmitterState001}
   &SpawnBurst_Instantaneous(Spawn Count=4, Use Spawn Probability=false, UseLoopCountLimit=false)
  ]
  particleSpawnScript [
   /Niagara/Modules/Spawn/Initialization/V2/InitializeParticle(
    Color=<rgba 0.1598 0.398 1 1>
    Lifetime=1.2
    Mesh Scale Mode=::NewEnumerator3
    Mesh Uniform Scale=&Divide_Float(A=$User.Radius, B=0)
    Position Offset=<15 0 0>
    Sprite Size Mode=::NewEnumerator3
    Uniform Sprite Size=50
    UsePositionOffset=false
    Write Material Random=true
   )
   &FN_SetSpawnGroup()
  ]
  particleUpdateScript [
   &ParticleState()
   &FN_BoneFlow(
    BoneID:NiagaraInt32=&ModuloInt(A=$Particles.SpawnGroup, B=4)
    Height:NiagaraFloat=&FloatFromCurve(FloatCurve=DataInterface() {propertyValues {Curve {keys [{interpMode=RCIM_Cubic, tangentMode=RCTM_User, time=0, value=0}, {interpMode=RCIM_Cubic, tangentMode=RCTM_User, time=1, value=1}]}}})
    SkeletalMesh=DataInterface() {propertyValues {
      DefaultMesh=null
      ExcludeBoneName=root
      OptionalComponentClass=null
      PreviewMesh=&Bites_Glider
      SkinningMode=SkinOnTheFly
      SoftSourceActor=null
      SourceMode=Default
      UvSetIndex=0
      WholeMeshLOD=-1
      bExcludeBone=false
      bReadDeformedGeometry=true
      bRequireCurrentFrameData=true
      ComponentTags []
      FilteredBones [JNT_lf_Base_root, JNT_lf_Base_4, JNT_lf_Branch4_3]
      FilteredSockets [
       R_Socket_Base
       R_Socket_A_01
       R_Socket_A_02
       R_Socket_Base
       R_Socket_B_01
       R_Socket_B_02
       R_Socket_Base
       R_Socket_C_01
       R_Socket_C_02
       R_Socket_Base
       R_Socket_D_01
       R_Socket_D_02
      ]
      MeshUserParameter {parameter {name=None, typeDefHandle {registeredTypeIndex=1}, varData []}}
      SamplingRegions []
     }}
    Use Sockets Instead=true
    Write Position Attribute=true
   ) {module=FN_BoneFlow002}
   &CameraOffset(Camera Offset Amount=80) {module=CameraOffset001}
   &DynamicMaterialParameters(
    Index 0 Param 1=1
    Index 0 Param 2=&Lerp_Float(
     A=1
     B=0.8
     Alpha=&FloatFromCurve(FloatCurve=DataInterface() {propertyValues {Curve {keys [
          {arriveTangent=5.1433, interpMode=RCIM_Cubic, time=0, value=0}
          {arriveTangent=0.9449, interpMode=RCIM_Cubic, leaveTangent=0.9449, tangentMode=RCTM_User, time=0.25, value=0.7}
          {arriveTangent=0.6574, interpMode=RCIM_Cubic, leaveTangent=0.6574, tangentMode=RCTM_User, time=1, value=1}
         ]}}})
    )
    Index 3 Param 1=&Multiply_Float(A=&MakeFloatFromVector(Vector=$Particles.Scale, Channel=::NewEnumerator0), B=90)
    Index 3 Param 2=&FloatFromCurve(FloatCurve=DataInterface() {propertyValues {Curve {keys [
         {arriveTangent=2.1437, interpMode=RCIM_Cubic, leaveTangent=2.1437, tangentMode=RCTM_User, time=0, value=0}
         {arriveTangent=1.3259, interpMode=RCIM_Cubic, leaveTangent=1.3259, tangentMode=RCTM_User, time=0.6, value=0.5}
         {interpMode=RCIM_Cubic, leaveTangent=0.2859, time=1, value=1}
        ]}}})
    Module.Param0WriteEnabled=true
    Module.Param0WriteEnabled3=true
    Module.Param1WriteEnabled=true
    Module.Param1WriteEnabled3=true
    Module.Param2WriteEnabled=false
    Module.Param2WriteEnabled3=false
    Module.Param3WriteEnabled=false
    Module.Param3WriteEnabled3=false
    Write Parameter Index 3=true
   )
   &ScaleSpriteSize(Uniform Curve Sprite Scale=DataInterface() {propertyValues {Curve {keys [
        {arriveTangent=0.0561, interpMode=RCIM_Cubic, leaveTangent=0.0561, tangentMode=RCTM_User, time=0, value=0}
        <at 0.2 1>
        {
         arriveTangent=-0.7272
         arriveTangentWeight=0.0565
         interpMode=RCIM_Cubic
         leaveTangent=-0.7272
         leaveTangentWeight=0.1797
         tangentMode=RCTM_User
         time=0.5
         value=0.875
        }
        {interpMode=RCIM_Cubic, leaveTangent=-0.9894, time=1, value=0.65}
       ]}}})
   &ScaleColor(
    Scale Alpha=&FloatFromCurve(
     FloatCurve=DataInterface() {propertyValues {Curve {keys [
         {arriveTangent=18.2404, interpMode=RCIM_Cubic, leaveTangent=18.2404, tangentMode=RCTM_User, time=0, value=1}
         {arriveTangent=-0.1486, interpMode=RCIM_Cubic, leaveTangent=-0.1486, tangentMode=RCTM_User, time=0.12, value=2}
         {arriveTangent=-0.1686, interpMode=RCIM_Cubic, leaveTangent=-0.1686, tangentMode=RCTM_User, time=0.7, value=1.875}
         {arriveTangent=-0.3515, interpMode=RCIM_Cubic, leaveTangent=-0.3515, tangentMode=RCTM_User, time=1, value=0}
        ]}}}
     Scale Curve=0.5
    )
    ScaleA=true
    ScaleRGB=true
   )
   /SetVariables(Particles.WaveColor:LinearColor=&LerpLinearColors(
     StartColor=<rgba 0.7113 0.6033 1 0>
     EndColor=<rgba 0.5788 1.0477 1.25 1>
     LerpFactor=&FloatFromCurve(FloatCurve=DataInterface() {propertyValues {Curve {keys [
          {arriveTangent=18.2701, interpMode=RCIM_Cubic, time=0, value=0}
          <at 0.15 1>
          {arriveTangent=29.7321, interpMode=RCIM_Cubic, leaveTangent=29.7321, tangentMode=RCTM_User, time=0.27, value=0.5}
          {arriveTangentWeight=0.0997, interpMode=RCIM_Cubic, leaveTangentWeight=0.3488, tangentMode=RCTM_User, time=0.3, value=1.1}
          {interpMode=RCIM_Cubic, leaveTangent=-2.4954, time=1, value=0}
         ]}}})
    )) {isSetParameters=true, module=SetVariables_#3}
   /SetVariables(Particles.LightColor=&MultiplyLinearColor(Linear Color=$Particles.Color, Scale Factor=&MakeVector4(X=1, Y=1, Z=1, W=15))) {isSetParameters=true, module=SetVariables_#4}
   /SetVariables(Particles.LightRadius=&FloatFromCurve(FloatCurve=DataInterface() {propertyValues {Curve {keys [<at 0.2 0.35 RCIM_Cubic>,<at 0.5 1>]}}}, Scale Curve=75)) {isSetParameters=true, module=SetVariables_#5}
  ]
  rendererClasses [&NiagaraSpriteRendererProperties]
  renderers [{rendererClass=&NiagaraSpriteRendererProperties}]
 }
 Right_Ribbons {
  emitterUpdateScript [
   &EmitterState(Life Cycle Mode=::NewEnumerator1, Loop Behavior=::NewEnumerator1, Loop Delay=0.5, Loop Duration=0.3, UseLoopDelay=true)
   &SpawnRate(SpawnRate=&Multiply_Float(A=30, B=20), Use Spawn Probability=false)
  ]
  particleSpawnScript [
   &FN_SetSpawnGroup(Inclusivity=true, Normalize Spawn Group=true, Particles Per Group=20, Pattern=::NewEnumerator0, Set Ribbon ID and Ribbon Link Order=true)
   &InitializeRibbon(
    Ribbon Width=&Multiply_Float(A=&RandomRangeFloat(Minimum=2.5, Maximum=3.5), B=2)
    UsePositionOffset=false
    Write Material Random=true
   )
  ]
  particleUpdateScript [
   &ParticleState()
   &FN_BoneFlow(
    Angle:NiagaraFloat=&RandomRangeFloat(Minimum=0, Maximum=1)
    BoneID:NiagaraInt32=&ModuloInt(A=$Particles.SpawnGroup, B=4)
    Height:NiagaraFloat=&Multiply_Float(A=$Particles.Normalized Spawn Group, B=1)
    Radius:NiagaraFloat=&RandomRangeFloat(Minimum=0, Maximum=10)
    SkeletalMesh=DataInterface() {propertyValues {
      DefaultMesh=null
      ExcludeBoneName=root
      OptionalComponentClass=null
      PreviewMesh=&Bites_Glider
      SkinningMode=SkinOnTheFly
      SoftSourceActor=null
      SourceMode=Default
      UvSetIndex=0
      WholeMeshLOD=-1
      bExcludeBone=false
      bReadDeformedGeometry=true
      bRequireCurrentFrameData=true
      ComponentTags []
      FilteredBones [JNT_lf_Base_root, JNT_lf_Base_4, JNT_lf_Branch4_3]
      FilteredSockets [
       R_Socket_Base
       R_Socket_A_01
       R_Socket_A_02
       R_Socket_Base
       R_Socket_B_01
       R_Socket_B_02
       R_Socket_Base
       R_Socket_C_01
       R_Socket_C_02
       R_Socket_Base
       R_Socket_D_01
       R_Socket_D_02
      ]
      MeshUserParameter {parameter {name=None, typeDefHandle {registeredTypeIndex=1}, varData []}}
      SamplingRegions []
     }}
    Use Sockets Instead=true
    Write Position Attribute=true
   ) {module=FN_BoneFlow001}
   /SetVariables(Particles.Position=&Add_Vector(
     A=Linked() {linkedVariable {name=Particles.Position, type=NiagaraPosition}}
     B=&Multiply_Vector(A=&ReturnCameraVector(Camera Vectors=::NewEnumerator1, Coordinate Space=::Simulation), B=<0 0 0>)
    )) {isSetParameters=true, module=SetVariables_#6}
   &ScaleRibbonWidth(Ribbon Width Scale=&FloatFromCurve(
     FloatCurve=DataInterface() {propertyValues {Curve {keys [
         {interpMode=RCIM_Cubic, tangentMode=RCTM_User, time=0, value=0}
         {arriveTangent=-5.6849, interpMode=RCIM_Cubic, leaveTangent=-5.6849, tangentMode=RCTM_User, time=0.1, value=0.4}
         {arriveTangent=-5.4632, interpMode=RCIM_Cubic, leaveTangent=-5.4632, tangentMode=RCTM_User, time=0.5, value=0.4}
         {arriveTangent=-2.6171, interpMode=RCIM_Cubic, leaveTangent=-2.6171, tangentMode=RCTM_User, time=0.8, value=0.5}
         {arriveTangent=-0.2025, interpMode=RCIM_Cubic, leaveTangent=-0.2025, tangentMode=RCTM_User, time=1, value=0.025}
        ]}}}
     CurveIndex=$Particles.Normalized Spawn Group
     Scale Curve=&Multiply_Float(
      A=&FloatFromCurve(
       FloatCurve=DataInterface() {propertyValues {Curve {keys [
           {arriveTangent=-0.082, interpMode=RCIM_Cubic, leaveTangent=-0.082, tangentMode=RCTM_User, time=0, value=0}
           {interpMode=RCIM_Cubic, tangentMode=RCTM_User, time=0.2, value=1.05}
           {interpMode=RCIM_Cubic, tangentMode=RCTM_User, time=0.5, value=1.2}
           {arriveTangent=-1.5059, interpMode=RCIM_Cubic, leaveTangent=-1.5059, tangentMode=RCTM_User, time=0.99, value=0}
          ]}}}
       Scale Curve=4
      )
      B=1
     )
    ))
   /SetVariables(Particles.RibbonFacing=$Output.FN_BoneFlow001.SplineBinormal) {isSetParameters=true, module=SetVariables_#7}
   &DynamicMaterialParameters(
    Index 0 Param 1=&FloatFromCurve(FloatCurve=DataInterface() {propertyValues {Curve {keys [<at 0 0 RCIM_Cubic>,<at 0.5 1>,<at 1 0 RCIM_Cubic>]}}})
    Module.Param0WriteEnabled=true
    Module.Param1WriteEnabled=true
    Module.Param2WriteEnabled=false
    Module.Param3WriteEnabled=false
   )
   &SolveForcesAndVelocity(Clamp Velocity=false, Limit Acceleration=false)
  ]
  rendererClasses [&NiagaraRibbonRendererProperties]
  renderers [{rendererClass=&NiagaraRibbonRendererProperties}]
 }
 Right_Tendrils {
  emitterUpdateScript [
   &EmitterState()
   &SpawnBurst_Instantaneous(Spawn Time=0.3, Use Spawn Probability=false, UseLoopCountLimit=false)
   &SpawnBurst_Instantaneous(Spawn Group=1, Spawn Time=0.25, Use Spawn Probability=false, UseLoopCountLimit=false) {module=SpawnBurst_Instantaneous001}
  ]
  particleSpawnScript [
   &InitializeParticle(
    Mesh Scale Mode=::NewEnumerator3
    Mesh Uniform Scale=&MaskFloatBySpawnGroup(Masked Float=1.25, Passthrough Float=1.5)
    UsePositionOffset=false
    Write Material Random=true
   )
   &InitialMeshOrientation(Mesh Orientation Mode=::NewEnumerator2, Use Rotation Vector=true) {module=InitialMeshOrientation002}
   /SetVariables(Particles.VisibilityTag=&MaskIntBySpawnGroup(Masked Int=0, Passthrough Int=1, Spawn Group Mask=0)) {isSetParameters=true, module=SetVariables_#8}
  ]
  particleUpdateScript [
   &ParticleState()
   &ScaleMeshSize(Scale Factor=&VectorFromFloat(Value=&FloatFromCurve(FloatCurve=DataInterface() {propertyValues {Curve {keys [
          {arriveTangent=-0.0856, interpMode=RCIM_Cubic, leaveTangent=-0.0856, tangentMode=RCTM_User, time=0, value=0.8}
          {arriveTangent=-0.0405, interpMode=RCIM_Cubic, leaveTangent=-0.0405, tangentMode=RCTM_User, time=0.3, value=1.35}
         ]}}}))) {module=ScaleMeshSize001}
   &ScaleMeshSize(Scale Factor=&FN_FrontEndVector3Modifier(Gameplay=<1 1 1>, Front End=<-1 1 1>))
   &SkeletalMeshLocation(
    Bone / Socket Index=1
    MaskSpawnGroup=false
    Mesh Sampling Type=::NewEnumerator5
    Orientation Sampling=::NewEnumerator1
    Sampled Position Offset=<10 -10 0>
    Skeletal Mesh=$User.TendrilsSKM
    Socket Sampling Mode=::NewEnumerator5
   )
   &CameraOffset(Camera Offset Amount=20)
   &DynamicMaterialParameters(
    Index 0 Param 1=&FloatFromCurve(FloatCurve=DataInterface() {propertyValues {Curve {keys [{arriveTangent=2.301, interpMode=RCIM_Cubic, time=0.8, value=0}, {interpMode=RCIM_Cubic, tangentMode=RCTM_User, time=1, value=1}]}}})
    Index 0 Param 3=1
    Index 1 Param 2=1
    Index 2 Param 1=&FloatFromCurve(FloatCurve=DataInterface() {propertyValues {Curve {keys [{tangentMode=RCTM_User, time=0, value=-1}, {tangentMode=RCTM_User, time=1, value=1}]}}})
    Index 2 Param 3=1
    Index 2 Param 4=2
    Module.Param0WriteEnabled=true
    Module.Param0WriteEnabled1=true
    Module.Param0WriteEnabled2=true
    Module.Param1WriteEnabled=true
    Module.Param1WriteEnabled1=true
    Module.Param1WriteEnabled2=true
    Module.Param2WriteEnabled=true
    Module.Param2WriteEnabled1=false
    Module.Param2WriteEnabled2=true
    Module.Param3WriteEnabled=true
    Module.Param3WriteEnabled1=false
    Module.Param3WriteEnabled2=true
    Write Parameter Index 1=true
    Write Parameter Index 2=true
   ) {module=DynamicMaterialParameters001}
  ]
  rendererClasses [&NiagaraMeshRendererProperties]
  renderers [
   {rendererClass=&NiagaraMeshRendererProperties, rendererIndex=0}
   {rendererClass=&NiagaraMeshRendererProperties, rendererIndex=1}
  ]
 }
 StartBlobs {
  emitterUpdateScript [
   &EmitterState()
   &SpawnBurst_Instantaneous(
    Spawn Count=&MakeCustomIntFromBool(Bool=$User.bIsFrontEnd, True INT=0, False INT=20)
    Spawn Time=0.25
    Use Spawn Probability=false
    UseLoopCountLimit=false
   )
  ]
  particleSpawnScript [
   /Niagara/Modules/Spawn/Initialization/InitializeParticle(
    Color=<rgba 0.276 0.276 0.276 1>
    Lifetime=&RandomRangeFloat(Minimum=0.25, Maximum=0.8)
    Module.Write Color=true
    Module.Write Lifetime=true
    Module.Write Mass=true
    Module.Write Material Random=true
    Module.Write Position=false
    Module.Write Scale=false
    Module.Write SpriteRotation=false
    Module.Write SpriteSize=true
    Position=&SimulationPosition()
    Sprite Size=&Vector2DFromFloat(Value=&RandomRangeFloat(Minimum=5, Maximum=8))
   )
   &FN_SetSpawnGroup(Normalize Spawn Group=true, Number of Spawn Groups=2, Particles Per Group=300, Set Ribbon ID and Ribbon Link Order=true)
   &SkeletalMeshLocation(
    Bone / Socket Index=$Particles.SpawnGroup
    Bone Sampling Mode=::NewEnumerator4
    MaskSpawnGroup=false
    Skeletal Mesh=DataInterface() {propertyValues {
      DefaultMesh=null
      ExcludeBoneName=root
      OptionalComponentClass=null
      PreviewMesh=&Bites_Glider
      SkinningMode=SkinOnTheFly
      SoftSourceActor=null
      SourceMode=Default
      UvSetIndex=0
      WholeMeshLOD=-1
      bExcludeBone=false
      bReadDeformedGeometry=true
      bRequireCurrentFrameData=true
      ComponentTags []
      FilteredBones [JNT_lf_Base_root, JNT_rt_Base_root]
      FilteredSockets []
      MeshUserParameter {parameter {name=None, typeDefHandle {registeredTypeIndex=1}, varData []}}
      SamplingRegions []
     }}
   ) {module=SkeletalMeshLocation002}
   &SkeletalMeshLocation(
    Bone / Socket Index=$Particles.SpawnGroup
    Bone Sampling Mode=::NewEnumerator4
    MaskSpawnGroup=false
    Skeletal Mesh=DataInterface() {propertyValues {
      DefaultMesh=null
      ExcludeBoneName=root
      OptionalComponentClass=null
      PreviewMesh=&Bites_Glider
      SkinningMode=SkinOnTheFly
      SoftSourceActor=null
      SourceMode=Default
      UvSetIndex=0
      WholeMeshLOD=-1
      bExcludeBone=false
      bReadDeformedGeometry=true
      bRequireCurrentFrameData=true
      ComponentTags []
      FilteredBones [JNT_lf_Base_2, JNT_rt_Base_2]
      FilteredSockets []
      MeshUserParameter {parameter {name=None, typeDefHandle {registeredTypeIndex=1}, varData []}}
      SamplingRegions []
     }}
   ) {module=SkeletalMeshLocation001}
   &ShapeLocation(MaskSpawnGroup=false, Non Uniform Scale=<1 0 1>, Override Seed=false, Sphere Radius=20, UseOwnerScale=true)
   &AddVelocity(
    Module.Use Falloff Distance=false
    Origin Offset=&MaskVectorBySpawnGroup(Masked Vector=<0 0 0>, Passthrough Vector=<-30 30 -30>)
    Override Seed=false
    Velocity Mode=::NewEnumerator1
    Velocity Origin:Vector3f=$Output.SkeletalMeshLocation002.SampledPositionWithoutOffset
    Velocity Speed=&RandomRangeFloat(Minimum=150, Maximum=300)
   )
  ]
  particleUpdateScript [
   &ParticleState()
   &SolveForcesAndVelocity(Clamp Velocity=false, Limit Acceleration=false, Speed Limit=1)
   &ScaleSpriteSize(Scale Factor=&Vector2DFromFloat(Value=&FloatFromCurve(FloatCurve=DataInterface() {propertyValues {Curve {keys [
          {arriveTangent=-0.0558, interpMode=RCIM_Cubic, leaveTangent=-0.0558, tangentMode=RCTM_User, time=0, value=0}
          {arriveTangent=0.044, interpMode=RCIM_Cubic, leaveTangent=0.044, tangentMode=RCTM_User, time=0.2, value=0.9847}
          {arriveTangent=-1.2308, interpMode=RCIM_Cubic, leaveTangent=-1.2308, time=0.7, value=0.8}
          {arriveTangent=-0.0405, interpMode=RCIM_Cubic, leaveTangent=-0.0405, tangentMode=RCTM_User, time=1, value=0}
         ]}}})))
   &CameraOffset(Camera Offset Amount=&RandomRangeFloat(Minimum=10, Maximum=20))
   &DynamicMaterialParameters(
    Index 0 Param 1=&FloatFromCurve(FloatCurve=DataInterface() {propertyValues {Curve {keys [
         {arriveTangent=130.5352, interpMode=RCIM_Cubic, leaveTangent=130.5352, tangentMode=RCTM_User, time=0, value=0}
         {arriveTangent=-0.891, interpMode=RCIM_Cubic, leaveTangent=-0.891, tangentMode=RCTM_User, time=0.03, value=1.7}
         <at 0.2 0.75>
         {tangentMode=RCTM_User, time=1, value=0}
        ]}}})
    Index 0 Param 2=5
    Index 0 Param 3=30
    Index 0 Param 4=7
    Index 1 Param 1=&FloatFromCurve(
     FloatCurve=DataInterface() {propertyValues {Curve {keys [
         {arriveTangent=-3.7492, interpMode=RCIM_Cubic, leaveTangent=-3.7492, tangentMode=RCTM_User, time=0, value=1}
         {arriveTangent=-0.7477, interpMode=RCIM_Cubic, leaveTangent=-0.7477, tangentMode=RCTM_User, time=0.3, value=0.3}
         {interpMode=RCIM_Cubic, leaveTangent=-1.9286, time=1, value=0}
        ]}}}
     CurveIndex=&FloatFromCurve(FloatCurve=DataInterface() {propertyValues {Curve {keys [
          {arriveTangent=130.5352, interpMode=RCIM_Cubic, leaveTangent=130.5352, tangentMode=RCTM_User, time=0, value=0}
          {arriveTangent=-0.891, interpMode=RCIM_Cubic, leaveTangent=-0.891, tangentMode=RCTM_User, time=0.03, value=1.7}
          <at 0.2 0.75>
          {tangentMode=RCTM_User, time=1, value=0}
         ]}}})
     Scale Curve=&RandomRangeFloat(Minimum=4, Maximum=4.5)
    )
    Module.Param0WriteEnabled=true
    Module.Param0WriteEnabled1=true
    Module.Param1WriteEnabled=true
    Module.Param1WriteEnabled1=false
    Module.Param2WriteEnabled=true
    Module.Param2WriteEnabled1=false
    Module.Param3WriteEnabled=true
    Module.Param3WriteEnabled1=false
    Write Parameter Index 1=true
   )
  ]
  rendererClasses [&NiagaraSpriteRendererProperties]
  renderers [{rendererClass=&NiagaraSpriteRendererProperties}]
 }
}
summary {
 systemName=NS_Glider_Bites_Deploy_02
 emitters [
  {emitterName=Left_Ribbons, rendererClasses [&NiagaraRibbonRendererProperties]}
  {emitterName=Left_Flares, rendererClasses [&NiagaraSpriteRendererProperties]}
  {emitterName=Left_Tendrils, rendererClasses [&NiagaraMeshRendererProperties]}
  {emitterName=Right_Ribbons, rendererClasses [&NiagaraRibbonRendererProperties]}
  {emitterName=Right_Tendrils, rendererClasses [&NiagaraMeshRendererProperties]}
  {emitterName=Right_Flares, rendererClasses [&NiagaraSpriteRendererProperties]}
  {emitterName=StartBlobs, rendererClasses [&NiagaraSpriteRendererProperties]}
  {emitterName=Left_Blobs002, rendererClasses [&NiagaraSpriteRendererProperties]}
  {emitterName=Right_Blobs003, rendererClasses [&NiagaraSpriteRendererProperties]}
 ]
 userVariables [
  User.bIsFrontEnd {type=NiagaraBool}
  User.bIsFrontEndPreview {type=NiagaraBool}
  User.BoneFlowSK {
   type=NiagaraDataInterfaceSkeletalMesh
   defaultValue {
    struct=NiagaraExt_VariableValue_DataInterface
    value {dataInterface=&self:NiagaraDataInterfaceSkeletalMesh_0}
   }
  }
  User.Color_Tint {type=LinearColor, description="", defaultValue {struct=LinearColor, value=<rgba 1 1 1 1>}}
  User.EmissiveIntensity {type=NiagaraFloat, defaultValue=0, description=""}
  User.Radius {type=NiagaraFloat, defaultValue=0, description=""}
  User.RibbonIDOffset {type=NiagaraInt32}
  User.SkeletalMesh {
   type=NiagaraDataInterfaceSkeletalMesh
   defaultValue {
    struct=NiagaraExt_VariableValue_DataInterface
    value {dataInterface=&self:NiagaraDataInterfaceSkeletalMesh_1}
   }
  }
  User.TendrilsSKM {
   type=NiagaraDataInterfaceSkeletalMesh
   defaultValue {
    struct=NiagaraExt_VariableValue_DataInterface
    value {dataInterface=&self:NiagaraDataInterfaceSkeletalMesh_2}
   }
  }
 ]
}
