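# Material function for the Glider Bites chute that adds animated liquid normals to incoming material attributes and also outputs them separately.
# A static switch chooses local triplanar or UV-panned normals, both warped by projected noise.
# The triplanar path blends opposite-scrolling projections through a sphere mask, mixing in a two-phase flowmap and flattening using a seam mask.
# Channel-masked strength and angle-corrected blending combine the result with the base normal.
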MaterialFunction /BRCosmetics/Gadgets/Assets/VinderTech_GliderChute/Glider_Bites/Materials/MF_GliderBitesFlowingNormals
graph {nodes [
  Constant3Vector() {constant=<rgba 0.2 0 0 1>}
  FunctionOutput(None=SetMaterialAttributes!(MaterialAttributes=NamedRerouteUsage(), Normal=NamedRerouteUsage()) {attributeSetTypes [#1]})
  FunctionOutput(None=NamedRerouteUsage()) {outputName=LiquidNormal}
  n1 = StaticSwitchParameter(True=NamedRerouteUsage(), False=NamedRerouteUsage()) {group=LiquidNormal, parameterName=LQN_UseLocalTriplanarNormals}
  n2 = ScalarParameter() {group=LiquidNormal, parameterName=LQN_NormalStrengthMin}
  n3 = ScalarParameter() {defaultValue=1, group=LiquidNormal, parameterName=LQN_NormalStrengthMax}
  n4 = FunctionInput() {bUsePreviewValueAsDefault=true, inputName=LiquidNormalMask, inputType=FunctionInput_Vector4, previewValue=<1 1 1 1>}
  n5 = ChannelMaskParameter(None=NamedRerouteDeclaration(None=@n4) {name=LiquidNormalMask, nodeColor=<rgba 0.5 0 1 1>}) {group=LiquidNormal, parameterName=LQN_MaskChannel}
  NamedRerouteDeclaration(None=LinearInterpolate(
    A=Constant3Vector() {constant=<rgba 0 0 1 0>}
    B=NamedRerouteDeclaration(None=@n1) {name=LiquidNormal_OUT, nodeColor=<rgba 0 1 0.5313 1>}
    Alpha=LinearInterpolate(A=@n2, B=@n3, Alpha=@n5.R)
   )) {name=Normal_OUT, nodeColor=<rgba 0.0078 0 1 1>}
  n6 = GetMaterialAttributes(None=FunctionInput() {inputType=FunctionInput_MaterialAttributes}) {attributeGetTypes [#1]}
  NamedRerouteDeclaration(None=@n6.Normal) {name=Normal_IN, nodeColor=<rgba 1 0 0.1406 1>}
  n7 = Multiply(A=TextureCoordinate(), B=ScalarParameter() {defaultValue=1.55, group=LiquidNormal, parameterName=FlowmapNormalTile})
  n8 = TextureSampleParameter2D() {group=LiquidNormal, parameterName="flowmapNM ", samplerType=SAMPLERTYPE_Normal, texture=&FlatNormalBlue}
  n9 = Multiply(A=ComponentMask(None=@n8.RGB) {g=true, r=true}, B=Constant() {r=1})
  n10 = Multiply(A=Time(), B=ScalarParameter() {defaultValue=0.3, group=LiquidNormal, parameterName=FlowMapSpeed})
  n11 = Frac(None=@n10)
  n12 = Add(A=@n7, B=Multiply(A=Multiply(A=@n9, B=@n11)))
  n13 = Frac(None=Add(A=@n10) {constB=0.5})
  n14 = Add(A=Multiply(A=@n9, B=Multiply(A=@n13)), B=@n7)
  n15 = Subtract(A=Multiply(A=@n11) {constB=2})
  n16 = NamedRerouteDeclaration(None=LinearInterpolate(
    A=TextureSample(UVs=@n12) {samplerType=SAMPLERTYPE_Normal, texture=/Game/Global/Textures/Default/Noise/T_FlowWarp_N}.RGB
    B=TextureSample(UVs=@n14) {samplerType=SAMPLERTYPE_Normal, texture=&T_FlowWarp_N}.RGB
    Alpha=Abs(None=@n15)
   )) {name=LiquidNormal_1, nodeColor=<rgba 1 0.7266 0 1>}
  n17 = MaterialFunctionCall() {materialFunction=&LocalPosition}
  n18 = Multiply(A=Abs(None=ScalarParameter() {defaultValue=256, group=LiquidNormal, parameterName=LQN_TriplanarNormalTextureSize})) {constA=-1, constB=-1}
  n19 = VectorParameter() {group=LiquidNormal, parameterName=LQN_TriplanarUVSpeed, defaultValue=<rgba 0 0.1 0 1>}
  n20 = Multiply(A=ComponentMask(None=NamedRerouteUsage()) {b=true, g=true, r=true}, B=@n19.RGB)
  n21 = Time()
  n22 = Add(
   A=Multiply(A=NamedRerouteUsage(), B=ScalarParameter() {group=LiquidNormal, parameterName=LQN_TriplanarUVWarpStrength})
   B=Add(A=Divide(A=@n17.Local Position, B=@n18), B=Multiply(A=@n20, B=@n21))
  )
  n23 = NamedRerouteUsage()
  n24 = TextureSample(UVs=ComponentMask(None=@n22) {b=true, r=true}, Tex=@n23) {samplerSource=SSM_Wrap_WorldGroupSettings, samplerType=SAMPLERTYPE_Normal}
  n25 = Transform(None=VertexNormalWS()) {transformSourceType=TRANSFORMSOURCE_World, transformType=TRANSFORM_Local}
  n26 = AppendVector(A=Multiply(A=@n24.R) {constB=-1}, B=Multiply(A=@n24.B, B=ComponentMask(None=@n25) {g=true}))
  n27 = Multiply(A=@n24.G) {constB=-1}
  n28 = ComponentMask(None=@n25) {r=true}
  n29 = TextureSample(UVs=ComponentMask(None=@n22) {b=true, g=true}, Tex=@n23) {samplerSource=SSM_Wrap_WorldGroupSettings, samplerType=SAMPLERTYPE_Normal}
  n30 = AppendVector(A=Multiply(A=@n28, B=@n29.B), B=Multiply(A=@n29.R) {constB=-1})
  n31 = Multiply(A=@n29.G) {constB=-1}
  n32 = Abs(None=@n28)
  n33 = ComponentMask(None=@n25) {b=true}
  n34 = Abs(None=@n33)
  n35 = TextureSample(UVs=ComponentMask(None=@n22) {g=true, r=true}, Tex=@n23) {samplerSource=SSM_Wrap_WorldGroupSettings, samplerType=SAMPLERTYPE_Normal}
  n36 = Multiply(A=@n35.RGB, B=Constant3Vector() {constant=<rgba -1 -1 1 0>})
  n37 = LinearInterpolate(
   A=LinearInterpolate(
    A=AppendVector(A=@n26, B=@n27)
    B=AppendVector(A=@n30, B=@n31)
    Alpha=MaterialFunctionCall(In=@n32, Contrast=@n34) {materialFunction=&CheapContrast}.Result
   )
   B=AppendVector(A=ComponentMask(None=@n36) {g=true, r=true}, B=Multiply(A=@n35.B, B=@n33))
   Alpha=MaterialFunctionCall(In=@n34, Contrast=ScalarParameter() {group=LiquidNormal, parameterName=LQN_TriplanarTransitionContrast}) {materialFunction=&CheapContrast}.Result
  )
  n38 = MaterialFunctionCall() {materialFunction=&LocalPosition}
  n39 = Multiply(A=Abs(None=ScalarParameter() {defaultValue=256, group=LiquidNormal, parameterName=LQN_TriplanarNormalTextureSize})) {constA=-1, constB=-1}
  n40 = Multiply(A=ComponentMask(None=NamedRerouteUsage()) {b=true, g=true, r=true}, B=Multiply(A=@n19.RGB) {constB=-1})
  n41 = Time()
  n42 = Add(
   A=Multiply(A=NamedRerouteUsage(), B=ScalarParameter() {group=LiquidNormal, parameterName=LQN_TriplanarUVWarpStrength})
   B=Add(A=Divide(A=@n38.Local Position, B=@n39), B=Multiply(A=@n40, B=@n41))
  )
  n43 = NamedRerouteUsage()
  n44 = TextureSample(UVs=ComponentMask(None=@n42) {b=true, r=true}, Tex=@n43) {samplerSource=SSM_Wrap_WorldGroupSettings, samplerType=SAMPLERTYPE_Normal}
  n45 = Transform(None=VertexNormalWS()) {transformSourceType=TRANSFORMSOURCE_World, transformType=TRANSFORM_Local}
  n46 = AppendVector(A=Multiply(A=@n44.R) {constB=-1}, B=Multiply(A=@n44.B, B=ComponentMask(None=@n45) {g=true}))
  n47 = Multiply(A=@n44.G) {constB=-1}
  n48 = ComponentMask(None=@n45) {r=true}
  n49 = TextureSample(UVs=ComponentMask(None=@n42) {b=true, g=true}, Tex=@n43) {samplerSource=SSM_Wrap_WorldGroupSettings, samplerType=SAMPLERTYPE_Normal}
  n50 = AppendVector(A=Multiply(A=@n48, B=@n49.B), B=Multiply(A=@n49.R) {constB=-1})
  n51 = Multiply(A=@n49.G) {constB=-1}
  n52 = Abs(None=@n48)
  n53 = ComponentMask(None=@n45) {b=true}
  n54 = Abs(None=@n53)
  n55 = TextureSample(UVs=ComponentMask(None=@n42) {g=true, r=true}, Tex=@n43) {samplerSource=SSM_Wrap_WorldGroupSettings, samplerType=SAMPLERTYPE_Normal}
  n56 = Multiply(A=@n55.RGB, B=Constant3Vector() {constant=<rgba -1 -1 1 0>})
  n57 = LinearInterpolate(
   A=LinearInterpolate(
    A=AppendVector(A=@n46, B=@n47)
    B=AppendVector(A=@n50, B=@n51)
    Alpha=MaterialFunctionCall(In=@n52, Contrast=@n54) {materialFunction=&CheapContrast}.Result
   )
   B=AppendVector(A=ComponentMask(None=@n56) {g=true, r=true}, B=Multiply(A=@n55.B, B=@n53))
   Alpha=MaterialFunctionCall(In=@n54, Contrast=ScalarParameter() {group=LiquidNormal, parameterName=LQN_TriplanarTransitionContrast}) {materialFunction=&CheapContrast}.Result
  )
  n58 = LinearInterpolate(
   A=Transform(None=Normalize(VectorInput=@n37)) {transformSourceType=TRANSFORMSOURCE_Local, transformType=TRANSFORM_Tangent}
   B=Transform(None=Normalize(VectorInput=@n57)) {transformSourceType=TRANSFORMSOURCE_Local, transformType=TRANSFORM_Tangent}
   Alpha=SphereMask(
    A=NamedRerouteUsage()
    B=VectorParameter() {group=LiquidNormal, parameterName=Actor_postion, defaultValue=<rgba 180 0 0 1>}.RGB
    Radius=ScalarParameter() {defaultValue=190, group=LiquidNormal, parameterName=ActorSphereMaskRadius}
   )
  )
  n59 = TextureSampleParameter2D() {group=LiquidNormal, parameterName=SeamMask, texture=&T_BlacksRGB}
  n60 = Saturate(None=Add(A=@n59.R, B=ScalarParameter() {defaultValue=0.1, group=LiquidNormal, parameterName=Flowmap_indensity}))
  n61 = ScalarParameter() {defaultValue=0.8, group=LiquidNormal, parameterName=SeamFlattenNormal}
  NamedRerouteDeclaration(None=MaterialFunctionCall(Normal=LinearInterpolate(A=@n16, B=@n58, Alpha=@n60), Flatness=LinearInterpolate(B=@n61, Alpha=@n59.R) {constB=0.8}) {materialFunction=&FlattenNormal}.Result) {name=TriplanarLiquidNormal, nodeColor=<rgba 1 0.4453 0 1>}
  n62 = FunctionInput() {bUsePreviewValueAsDefault=true, inputName=Alpha, inputType=FunctionInput_Scalar, previewValue=<1 0 0 1>}
  NamedRerouteDeclaration(None=@n62) {name=Alpha_IN, nodeColor=<rgba 1 0 0.9844 1>}
  n63 = NamedRerouteUsage()
  n64 = NamedRerouteUsage()
  NamedRerouteDeclaration(None=LinearInterpolate(
    A=@n63
    B=MaterialFunctionCall(BaseNormal=@n63, AdditionalNormal=@n64) {materialFunction=&BlendAngleCorrectedNormals}.Result
    Alpha=NamedRerouteUsage()
   )) {name=Normal, nodeColor=<rgba 1 0.0703 0 1>}
  NamedRerouteDeclaration(None=@n6.MaterialAttributes) {name=Material_IN, nodeColor=<rgba 0 1 0.5547 1>}
  NamedRerouteDeclaration(None=VertexInterpolator(VS=PreSkinnedNormal()).PS) {name=PS_Normal, nodeColor=<rgba 1 0 0.8437 1>}
  NamedRerouteDeclaration(None=VertexInterpolator(VS=PreSkinnedPosition()).PS) {name=PS_Position, nodeColor=<rgba 1 0 0.1641 1>}
  n65 = TextureObjectParameter() {
   group=LiquidNormal
   parameterName=LQN_WarpTexture
   samplerType=SAMPLERTYPE_Normal
   texture=/Game/Global/Textures/Default/Noise/T_FlowWarp_Noise
  }
  n66 = NamedRerouteUsage()
  n67 = Multiply(A=Time(), B=VectorParameter() {group=LiquidNormal, parameterName=LQN_WarpUVSpeed, defaultValue=<rgba 0 0 0 1>}.RGB)
  NamedRerouteDeclaration(None=MaterialFunctionCall(
    TextureObject=@n65
    TextureSize=ScalarParameter() {defaultValue=128, group=LiquidNormal, parameterName=LQN_WarpTextureSize}
    WorldPosition=Add(A=@n66, B=@n67)
    World Space Normal=NamedRerouteUsage()
    ProjectionTransitionContrast=ScalarParameter() {defaultValue=1, group=LiquidNormal, parameterName=LQN_WarpProjectionContrast, sliderMax=1}
   ) {materialFunction=&WorldAlignedTexture}.XYZ Texture) {name=UVWarp, nodeColor=<rgba 0.3828 0 1 1>}
  n68 = Multiply(A=ComponentMask(None=NamedRerouteUsage()) {g=true, r=true}, B=ScalarParameter() {group=LiquidNormal, parameterName=LQN_UVWarpStrength})
  n69 = ComponentMask(None=NamedRerouteUsage()) {g=true, r=true}
  n70 = ComponentMask(None=VectorParameter() {group=LiquidNormal, parameterName=LQN_UVSpeed, defaultValue=<rgba 0 0.1 0 1>}.RGB) {g=true, r=true}
  n71 = Add(
   A=Multiply(A=TextureCoordinate(), B=ScalarParameter() {defaultValue=1, group=LiquidNormal, parameterName=LQN_UVTile})
   B=Multiply(A=Multiply(A=@n69, B=@n70), B=Time())
  )
  n72 = TextureObjectParameter() {
   group=LiquidNormal
   parameterName=LQN_NormalTexture
   samplerType=SAMPLERTYPE_Normal
   texture=&T_FlowWarp_N
  }
  NamedRerouteDeclaration(None=TextureSample(UVs=Add(A=@n68, B=@n71), Tex=NamedRerouteDeclaration(None=@n72) {name=NormalTexture, nodeColor=<rgba 1 0.6094 0 1>}).RGB) {name=LiquidNormal, nodeColor=<rgba 1 0.7266 0 1>}
  PreSkinnedPosition()
 ]}
