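# Master material for the Glider Bites glider, a skeletal-mesh Default Lit surface with flowing liquid normals and a subtle thin-film tint.
# Character PreFX output feeds the flowing-normals function, masked by a sphere in pre-skinned local space, before the PostFX function finishes it.
# Base color is lightened, then multiplied by a washed-out, desaturated spectrum texture sampled with Fresnel on a transformed normal.
# Roughness is remapped between two parameters using the PreFX roughness.
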Material /BRCosmetics/Gadgets/Assets/VinderTech_GliderChute/Glider_Bites/Materials/M_GliderBites_Master
graph {
 properties {
  bUseMaterialAttributes=true
  bUsedWithMeshDeformer=true
  bUsedWithSkeletalMesh=true
  pixelDepthOffsetMode=PDOM_Legacy
  usedShadingModels=Default Lit
 }
 nodes [
  n1 = MaterialFunctionCall() {materialFunction=&BaseFN_CharacterFunction_PreFX}
  n2 = GetMaterialAttributes(None=@n1.Result) {attributeGetTypes [#1, #2]}
  n3 = MaterialFunctionCall!(
   Alpha=SphereMask(
    A=@n1.Pre skinned Local Position
    B=Constant3Vector()
    Radius=ScalarParameter() {defaultValue=20, group=PickaxeBites, parameterName=liquidNormalRadius}
    Hardness=ScalarParameter() {defaultValue=1, group=PickaxeBites, parameterName=liquidNormalHardness}
   ) {attenuationRadius=20, hardnessPercent=1}
   LiquidNormalMask=Constant4Vector() {constant=<rgba 1 1 1 1>}.RGBA
   In=@n2.MaterialAttributes
  ) {materialFunction=&/MF_GliderBitesFlowingNormals}
  n4 = ScalarParameter() {defaultValue=0.23, group=PickaxeBites, parameterName=RoughnessMin_a}
  n5 = ScalarParameter() {defaultValue=0.4, group=PickaxeBites, parameterName=RoughnessMax_B}
  MaterialFunctionCall(
   AO=@n1.AO
   Distance Blend=@n1.AO
   Fresnel=@n1.AO
   Gmap=@n1.AO
   Material=SetMaterialAttributes!(
    MaterialAttributes=@n3.Result
    Base Color=NamedRerouteUsage()
    Roughness=LinearInterpolate(A=@n4, B=@n5, Alpha=@n2.Roughness) {constA=0.23, constB=0.5}
   ) {attributeSetTypes [#1, #2]}
  ) {materialFunction=&BaseFN_CharacterFunction_PostFX}
  NamedRerouteDeclaration(None=@n3.LiquidNormal) {name=liquidNormal, nodeColor=<rgba 0.9219 0 1 1>}
  n6 = Constant3Vector() {constant=<rgba 0.9323 0.9323 0.9323 0>}
  n7 = ScalarParameter() {defaultValue=0.5, group=PickaxeBites, parameterName=basecolor_brightness}
  n8 = Constant() {r=3}
  n9 = Transform(None=NamedRerouteUsage())
  n10 = LinearInterpolate(
   A=TextureSampleParameter2D(UVs=Fresnel(ExponentIn=@n8, Normal=@n9)) {parameterName=Param, texture=/Game/Global/Textures/Default/Thinfilm/T_ThinFilm_Spectrum_COLOR}.RGB
   Alpha=Constant() {r=0.85}
  )
  n11 = Constant() {r=0.7}
  NamedRerouteDeclaration(None=Multiply(A=LinearInterpolate(A=@n2.Base Color, B=@n6, Alpha=@n7), B=Desaturation(None=@n10, Fraction=@n11))) {name=Base color, nodeColor=<rgba 0.7344 1 0 1>}
 ]
}
